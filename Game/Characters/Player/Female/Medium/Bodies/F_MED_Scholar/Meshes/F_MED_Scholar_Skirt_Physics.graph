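# Physics asset for the skirt of the medium female Scholar player body, with capsule bodies on eight dynamic frill bones, the pelvis, thighs and small helper bones.
# Side and back-corner frills hang from sub-thigh bones, front corners from driver bones, center back from pelvis, center front from a driven flap bone.
# Frill joints have angular limits, tighter between side and front neighbors, while thigh-to-pelvis, driver and flap-to-pelvis joints rotate freely.
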PhysicsAsset /Game/Characters/Player/Female/Medium/Bodies/F_MED_Scholar/Meshes/F_MED_Scholar_Skirt_Physics
bodies {
 Flap_Front_Driven_Bone [{shapeType=Capsule, length=0.505, radius=0.505, rotation=<rot 0 -90 0>}]
 driver_frill_frt_l [{shapeType=Capsule, length=0.5, radius=1.25}]
 driver_frill_frt_r [{shapeType=Capsule, length=0.505, radius=0.505}]
 dyn_frill_bck [{shapeType=Capsule, length=5.2413, radius=2.8791, center=<0 0 -4.747>}]
 dyn_frill_bck_l [{shapeType=Capsule, length=5.0675, radius=2, center=<-3.4098 -0.0001 0>, rotation=<rot 90 0 0>}]
 dyn_frill_bck_r [{shapeType=Capsule, length=4.8399, radius=2, center=<3.8035 0 0>, rotation=<rot 90 0 0>}]
 dyn_frill_frt [{shapeType=Capsule, length=5, radius=2.5412, center=<0 0 -3.5396>}]
 dyn_frill_frt_l [{shapeType=Capsule, length=5.5298, radius=2, center=<-3.8324 -0.0001 0>, rotation=<rot 90 0 0>}]
 dyn_frill_frt_r [{shapeType=Capsule, length=4.2308, radius=2, center=<3.2716 -0.0001 0>, rotation=<rot 90 0 0>}]
 dyn_frill_side_l [{shapeType=Capsule, length=4.5276, radius=2, center=<-3.0869 0.0001 0>, rotation=<rot 90 0 0>}]
 dyn_frill_side_r [{shapeType=Capsule, length=4.4908, radius=2, center=<3.586 0.0002 0>, rotation=<rot 90 0 0>}]
 pelvis [{shapeType=Capsule, length=6.4689, radius=3.0903}]
 sub_thigh_l [{shapeType=Capsule, length=0.505, radius=0.505, rotation=<rot 0 -90 0>}]
 sub_thigh_r [{shapeType=Capsule, length=0.505, radius=0.505, rotation=<rot 0 -90 0>}]
 thigh_l [{shapeType=Capsule, length=9, radius=7.6699, center=<-15.0889 0.3999 0.2753>, rotation=<rot 0 95.6724 -89.5096>}]
 thigh_r [{shapeType=Capsule, length=9.5, radius=7.6548, center=<14.9883 -0.4197 -0.3821>, rotation=<rot 0 -84.0894 -89.5258>}]
}
constraints [
 {bone1Name=sub_thigh_l, bone2Name=thigh_l, swing1LimitDegrees=45, swing2LimitDegrees=45, twistLimitDegrees=45}
 {bone1Name=sub_thigh_r, bone2Name=thigh_r, swing1LimitDegrees=45, swing2LimitDegrees=45, twistLimitDegrees=45}
 {bone1Name=dyn_frill_bck_l, bone2Name=sub_thigh_l, swing1LimitDegrees=45, swing2LimitDegrees=45, twistLimitDegrees=45}
 {bone1Name=dyn_frill_frt_l, bone2Name=driver_frill_frt_l, swing1LimitDegrees=45, swing2LimitDegrees=30, twistLimitDegrees=10}
 {bone1Name=dyn_frill_side_l, bone2Name=sub_thigh_l, swing1LimitDegrees=45, swing2LimitDegrees=45, twistLimitDegrees=15}
 {bone1Name=dyn_frill_bck_r, bone2Name=sub_thigh_r, swing1LimitDegrees=45, swing2LimitDegrees=45, twistLimitDegrees=45}
 {bone1Name=dyn_frill_frt_r, bone2Name=driver_frill_frt_r, swing1LimitDegrees=45, swing2LimitDegrees=30, twistLimitDegrees=10}
 {bone1Name=dyn_frill_side_r, bone2Name=sub_thigh_r, swing1LimitDegrees=45, swing2LimitDegrees=45, twistLimitDegrees=15}
 {
  bone1Name=thigh_l
  bone2Name=pelvis
  swing1LimitDegrees=45
  swing1Motion=Free
  swing2LimitDegrees=45
  swing2Motion=Free
  twistLimitDegrees=45
  twistMotion=Free
 }
 {
  bone1Name=thigh_r
  bone2Name=pelvis
  swing1LimitDegrees=45
  swing1Motion=Free
  swing2LimitDegrees=45
  swing2Motion=Free
  twistLimitDegrees=45
  twistMotion=Free
 }
 {bone1Name=dyn_frill_frt, bone2Name=Flap_Front_Driven_Bone, swing1LimitDegrees=45, swing2LimitDegrees=45, twistLimitDegrees=45}
 {bone1Name=dyn_frill_bck, bone2Name=pelvis, swing1LimitDegrees=45, swing2LimitDegrees=45, twistLimitDegrees=45}
 {bone1Name=dyn_frill_side_l, bone2Name=dyn_frill_frt_l, swing1LimitDegrees=20, swing2LimitDegrees=20, twistLimitDegrees=10}
 {bone1Name=dyn_frill_side_r, bone2Name=dyn_frill_frt_r, swing1LimitDegrees=20, swing2LimitDegrees=20, twistLimitDegrees=10}
 {
  bone1Name=Flap_Front_Driven_Bone
  bone2Name=pelvis
  swing1LimitDegrees=45
  swing1Motion=Free
  swing2LimitDegrees=45
  swing2Motion=Free
  twistLimitDegrees=45
  twistMotion=Free
 }
 {
  bone1Name=driver_frill_frt_l
  bone2Name=sub_thigh_l
  swing1LimitDegrees=45
  swing1Motion=Free
  swing2LimitDegrees=45
  swing2Motion=Free
  twistLimitDegrees=45
  twistMotion=Free
 }
 {
  bone1Name=driver_frill_frt_r
  bone2Name=sub_thigh_r
  swing1LimitDegrees=45
  swing1Motion=Free
  swing2LimitDegrees=45
  swing2Motion=Free
  twistLimitDegrees=45
  twistMotion=Free
 }
]
